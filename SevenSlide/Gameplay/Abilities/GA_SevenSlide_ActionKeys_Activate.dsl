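; Gameplay ability for the SevenSlide item that manages the HUD action keys for activation while the item is in inventory.
; Chained wait tasks watch the powered-on tag, harvester tag and hover-activatable query, each refreshing the keys on change.
; The refresh shows and activates the Crouch keys when powered on without turn-off mode tags, and hides Target while the harvesting tool is equipped.
; Ending the ability ends the stored tasks and refreshes again.
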(event Ability|EventActivateAbility
  (bind _poweredontag (Variables|Default|GetPoweredOnTag))
  (bind _harvestertag (Variables|Default|GetHarvesterTag))
  (bind _hoveractivatablequery (Variables|Default|GetHoverActivatableQuery))
  (|CallParentFunction)
  (Default|UpdateActivateHUDActionKey) ; Make sure it is not visible
  ;[0 1 0] Activate Ability - Activated when in inventory
  (bind _asynctaskproxy (Ability|Tasks|WaitGameplayTagAdd _poweredontag)
    (:then
      (Variables|Default|SetAddTask _asynctaskproxy)
      (bind _asynctaskproxy_2 (Ability|Tasks|WaitGameplayTagRemove _poweredontag)
        (:then
          (Variables|Default|SetRemoveTask _asynctaskproxy_2)
          (bind _asynctaskproxy_3 (Ability|Tasks|WaitGameplayTagAdd _harvestertag)
            (:then
              (Variables|Default|SetAddTaskHarvester _asynctaskproxy_3)
              (bind _asynctaskproxy_4 (Ability|Tasks|WaitGameplayTagRemove _harvestertag)
                (:then
                  (Variables|Default|SetRemoveTaskHarvester _asynctaskproxy_4)
                  (bind _asynctaskproxy_5 (Ability|Tasks|WaitGameplayTagQuery _hoveractivatablequery)
                    (:then
                      (Variables|Default|SetShowHoverTrueTask _asynctaskproxy_5)
                      (bind _asynctaskproxy_6 (Ability|Tasks|WaitGameplayTagQuery _hoveractivatablequery 0 "WhenFalse")
                        (:then
                          (Variables|Default|SetShowHoverFalseTask _asynctaskproxy_6))
                        (:Triggered
                          (Default|UpdateActivateHUDActionKey))))
                    (:Triggered
                      (Default|UpdateActivateHUDActionKey) ; Make sure it is not visible
                    )))
                (:Removed
                  (Default|UpdateActivateHUDActionKey) ; Make sure it is not visible
                )))
            (:Added
              (Default|UpdateActivateHUDActionKey) ; Make sure it is not visible
            )))
        (:Removed
          (Default|UpdateActivateHUDActionKey) ; Make sure it is not visible
        )))
    (:Added
      (Default|UpdateActivateHUDActionKey) ; Make sure it is not visible
    )))

(event Ability|EventOnEndAbility
  (params (bWasCancelled :type Boolean))
  ;[1 0.25 0] End Ability - Hide HUD Action keys and end async task
  (|CallParentFunction bWasCancelled)
  (Utilities|FlowControl|Sequence
    (:then_0
      (bind _showhovertruetask (Variables|Default|GetShowHoverTrueTask)
        (:then
          (GameplayTasks|EndTask _showhovertruetask))
        (:else)))
    (:then_1
      (bind _showhoverfalsetask (Variables|Default|GetShowHoverFalseTask)
        (:then
          (GameplayTasks|EndTask _showhoverfalsetask))
        (:else)))
    (:then_2
      (GameplayTasks|EndTask (Variables|Default|GetAddTask))
      (GameplayTasks|EndTask (Variables|Default|GetAddTaskHarvester)))
    (:then_3
      (GameplayTasks|EndTask (Variables|Default|GetRemoveTask))
      (GameplayTasks|EndTask (Variables|Default|GetRemoveTaskHarvester)))
    (:then_4
      (Default|UpdateActivateHUDActionKey))))

(fn UpdateActivateHUDActionKey
  (bind _turnoffmodes (Variables|Default|GetTurnOffModes))
  (bind _returnvalue (CallFunction|GetAvatarasFortPlayerPawn))
  (bind _returnvalue_2 (GameplayTags|HasAnyMatchingGameplayTags _returnvalue _turnoffmodes))
  (bind _returnvalue_3 (not _returnvalue_2))
  (bind _poweredontag (Variables|Default|GetPoweredOnTag))
  (bind _returnvalue_4 (GameplayTags|HasMatchingGameplayTag _returnvalue _poweredontag))
  (bind _returnvalue_5 (and _returnvalue_4 _returnvalue_3))
  (CallFunction|SetActionDisplayVisibility "Crouch" _returnvalue_5)
  (CallFunction|SetActionDisplayVisibility "GamepadCrouch" _returnvalue_5)
  (CallFunction|SetActionDisplayActive _returnvalue_5)
  (CallFunction|SetActionDisplayVisibility "Target" (not (GameplayTags|HasMatchingGameplayTag (CallFunction|GetAvatarasFortPlayerPawn) "(TagName=\"Granted.ItemEquipped.Melee.Harvest\")"))))
