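; Gameplay ability that plays a directional exit montage when leaving a Seven Slide and applies a movement speed exit boost, gated by a hotfix scalable float.
; It picks forward, right, left or back by the highest dot product of normalized pawn velocity against the pawn's axes.
; A jump tag, cleared after reading, selects jump variants of each montage.
; The ability ends right after starting the montage and applying the boost effect.
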(event Ability|EventActivateAbility
  (bind _returnvalue (CallFunction|GetAvatarasFortPlayerPawn))
  (bind _returnvalue_2 (CallFunction|GetAvatarasFortPlayerPawn))
  (bind _returnvalue_3 (Transformation|GetActorRightVector _returnvalue_2))
  (bind _returnvalue_4 (Math|Vector|DotProduct (Math|Vector|Normalize (Transformation|GetVelocity (CallFunction|GetAvatarasFortPlayerPawn))) _returnvalue_3))
  (bind _returnvalue_5 (CallFunction|GetAvatarasFortPlayerPawn))
  (bind _returnvalue_6 (Transformation|GetActorRightVector _returnvalue_5))
  (bind _returnvalue_7 (* _returnvalue_6 -1.0))
  (bind _returnvalue_8 (Math|Vector|DotProduct (Math|Vector|Normalize (Transformation|GetVelocity (CallFunction|GetAvatarasFortPlayerPawn))) _returnvalue_7))
  (bind _returnvalue_9 (CallFunction|GetAvatarasFortPlayerPawn))
  (bind _returnvalue_10 (Transformation|GetVelocity _returnvalue_9))
  (bind _returnvalue_11 (Math|Vector|Normalize _returnvalue_10))
  (bind _returnvalue_12 (Transformation|GetActorForwardVector _returnvalue))
  (bind _returnvalue_13 (* _returnvalue_12 -1.0))
  (bind _returnvalue_14 (Math|Vector|DotProduct _returnvalue_11 _returnvalue_13))
  (if (> (ScalableFloat|GetValueasDouble (Variables|Default|GetAllowExitAnimsHotfix)) 0)
    (Variables|Default|SetIsComingFromJump (GameplayTags|HasMatchingGameplayTag (CallFunction|GetAvatarasFortPlayerPawn) (Variables|Default|GetJumpAnimTag)))
    (Ability|GameplayEffect|RemoveLooseGameplayTags (CallFunction|GetAvatarasFortPlayerPawn) (Variables|Default|GetJumpAnimTagContainer))
    (if (Variables|Default|GetIsComingFromJump)
      (Variables|Default|SetMontageToPlay (Variables|Default|GetForwardMontage_Jump))
      (else
        (Variables|Default|SetMontageToPlay (Variables|Default|GetForwardMontage))))
    (Variables|Default|SetMostAlignedValue (Math|Vector|DotProduct _returnvalue_11 (Transformation|GetActorForwardVector (CallFunction|GetAvatarasFortPlayerPawn))))
    (if (> _returnvalue_4 (Variables|Default|GetMostAlignedValue))
      (if (Variables|Default|GetIsComingFromJump)
        (Variables|Default|SetMontageToPlay (Variables|Default|GetRightMontage_Jump))
        (else
          (Variables|Default|SetMontageToPlay (Variables|Default|GetRightMontage))))
      (Variables|Default|SetMostAlignedValue _returnvalue_4)
      (else
        _))
    (if (> _returnvalue_8 (Variables|Default|GetMostAlignedValue))
      (if (Variables|Default|GetIsComingFromJump)
        (Variables|Default|SetMontageToPlay (Variables|Default|GetLeftMontage_Jump))
        (else
          (Variables|Default|SetMontageToPlay (Variables|Default|GetLeftMontage))))
      (Variables|Default|SetMostAlignedValue _returnvalue_8)
      (else
        _))
    (if (> _returnvalue_14 (Variables|Default|GetMostAlignedValue))
      (if (Variables|Default|GetIsComingFromJump)
        (Variables|Default|SetMontageToPlay (Variables|Default|GetBackMontage_Jump))
        (else
          (Variables|Default|SetMontageToPlay (Variables|Default|GetBackMontage))))
      (Variables|Default|SetMostAlignedValue _returnvalue_14)
      (else
        _))
    (Ability|Tasks|PlayMontageAndWaitNotify "SevenSliderRoll" (Variables|Default|GetMontageToPlay) 1.0 "Start" false)
    (GameplayEffects|ApplyGameplayEffectToSelf :self (Ability|GetAbilitySystemComponent (CallFunction|GetAvatarasFortPlayerPawn)) :GameplayEffectClass "/SevenSlide/Gameplay/Effects/GE_SevenSlider_MovementSpeedExitBoost.GE_SevenSlider_MovementSpeedExitBoost_C" :Level 0.0)
    (Ability|EndAbility)
    (else
      (Ability|EndAbility))))

(event Ability|EventOnEndAbility
  (params (bWasCancelled :type Boolean)))
