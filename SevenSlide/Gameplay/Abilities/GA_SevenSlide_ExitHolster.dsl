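; Gameplay ability that handles exiting the Seven Slide, either holstering or triggering the exit animation.
; On activation it removes the starting loose tag from the Fort player pawn, then checks a scalable-float holster toggle.
; When enabled, it applies the holster effect and removes the pause-slide-when-jumping effect; otherwise it sends a play-exit-animation gameplay event.
; Both paths end the ability immediately, and the end-ability event has no logic.
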(event Ability|EventActivateAbility
  (bind _returnvalue (CallFunction|GetAvatarasFortPlayerPawn))
  (bind _returnvalue_2 (Ability|GetAbilitySystemComponent _returnvalue))
  (Ability|GameplayEffect|RemoveLooseGameplayTags _returnvalue (Variables|Default|GetStartingTag) true)
  (if (> (ScalableFloat|GetValueatLevel (Variables|Default|GetShouldSevenSlideHolsterOnExit)) 0)
    (Fortnite|ApplyGameplayEffectsToActor _returnvalue (Variables|Default|GetHolsterGameplayEffect) (Transformation|GetActorLocation _returnvalue) false 1 _returnvalue_2)
    (GameplayEffects|RemoveActiveGameplayEffectbySourceEffect _returnvalue_2 "/SevenSlide/Gameplay/Effects/GE_BoostSeven_PauseSlideWhenJumping.GE_BoostSeven_PauseSlideWhenJumping_C" _returnvalue_2 -1)
    (Ability|EndAbility)
    (else
      (Ability|SendGameplayEvent "(TagName=\"sevenslide.playexitanim\")")
      (Ability|EndAbility))))

(event Ability|EventOnEndAbility
  (params (bWasCancelled :type Boolean)))
