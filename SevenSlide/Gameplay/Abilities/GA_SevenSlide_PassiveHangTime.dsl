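; Passive gameplay ability that toggles hang time for the SevenSlide mechanic on the avatar pawn.
; Add and remove tag tasks track sliding, targeting and manual hang-timing, each calling EvaluateHangTime.
; While the PoweredOn tag is present, hang time requires sliding plus targeting or manual input; changes toggle hang-time and reticle-shrink effects, camera lens effect and camera modifier.
; Ending the ability clears only the camera visuals.
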(event Ability|EventActivateAbility
  (bind _outfortplayerpawn (CallFunction|GetAvatarasValidFortPlayerPawn)
    (:ValidFortPlayerPawn
      (Variables|Default|SetAvatarFortPawn _outfortplayerpawn)
      (Utilities|FlowControl|Sequence
        (:then_0
          (Ability|Tasks|WaitGameplayTagRemove (Variables|Default|GetIsTargetingTag)
            (:Removed
              (Variables|Default|SetIsTargetting false)
              (CallFunction|EvaluateHangTime))))
        (:then_1
          (Ability|Tasks|WaitGameplayTagRemove (Variables|Default|GetIsSevenSlidingTag)
            (:Removed
              (Variables|Default|SetIsSevenSliding false)
              (CallFunction|EvaluateHangTime))))
        (:then_2
          (Ability|Tasks|WaitGameplayTagRemove (Variables|Default|GetIsManualHangTimingTag)
            (:Removed
              (Variables|Default|SetIsManualHangTiming false)
              (CallFunction|EvaluateHangTime))))
        (:then_3
          (Ability|Tasks|WaitGameplayTagAdd (Variables|Default|GetIsTargetingTag)
            (:Added
              (Variables|Default|SetIsTargetting true)
              (CallFunction|EvaluateHangTime))))
        (:then_4
          (Ability|Tasks|WaitGameplayTagAdd (Variables|Default|GetIsSevenSlidingTag)
            (:Added
              (Variables|Default|SetIsSevenSliding true)
              (CallFunction|EvaluateHangTime))))
        (:then_5
          (Ability|Tasks|WaitGameplayTagAdd (Variables|Default|GetIsManualHangTimingTag)
            (:Added
              (Variables|Default|SetIsManualHangTiming true)
              (CallFunction|EvaluateHangTime))))))))

(event Ability|EventOnEndAbility
  (params (bWasCancelled :type Boolean))
  (Default|RemoveHangtimeCameraLensEffect))

(fn AddHangtimeCameraLensEffect
  (bind _returnvalue (Camera|AddGenericCameraLensEffect (Variables|/Script/Engine.PlayerController|GetPlayerCameraManager (Fortnite|GetFortPlayerControllerfromActor (Variables|Default|GetAvatarFortPawn))) (Variables|Default|GetLens%20Effect%20Emitter%20Class)))
  (Variables|Default|SetCameraLensEffect _returnvalue)
  (bind _outmodifier (Class|WeaponFunctionLibrary|FindorAddCameraModifierforPawn (Variables|Default|GetCamera%20Modifier%20Class) (Variables|Default|GetAvatarFortPawn)))
  (bind _asseven_slide_camera_modifier_hang_time (Utilities|Casting|CastToSevenSlide_CameraModifier_HangTime _outmodifier))
  (Variables|/SevenSlide/Vfx/SevenSlide_CameraModifier_HangTime.SevenSlide_CameraModifier_HangTime_C|SetIsActive true _asseven_slide_camera_modifier_hang_time))

(fn EvaluateHangTime
  (if (GameplayTags|HasMatchingGameplayTag (CallFunction|GetAvatarasFortPlayerPawn) (Variables|Default|GetPoweredOn))
    (bind _output_get (LocalVariables|SetShouldBeHangTiming (and (Variables|Default|GetIsSevenSliding) (or (Variables|Default|GetIsTargetting) (Variables|Default|GetIsManualHangTiming)))))
    (if (not (== _output_get (Variables|Default|GetIsHangTiming)))
      (bind _output_get_2 (Variables|Default|SetIsHangTiming (LocalVariables|GetShouldBeHangTiming)))
      (if _output_get_2
        (bind _returnvalue (Ability|ApplyGameplayEffectToOwner "/SevenSlide/Gameplay/Effects/GE_SevenSlider_HangTime.GE_SevenSlider_HangTime_C"))
        (Variables|Default|SetHangtimeEffect _returnvalue)
        (Default|AddHangtimeCameraLensEffect)
        (bind _returnvalue_2 (Ability|ApplyGameplayEffectToOwner "/SevenSlide/Gameplay/Effects/GE_SevenSlider_ReticleShrink.GE_SevenSlider_ReticleShrink_C"))
        (Variables|Default|SetReticleShrinkEffect _returnvalue_2)
        (else
          (Ability|RemoveGameplayEffectFromOwnerWithHandle (Variables|Default|GetHangtimeEffect))
          (Default|RemoveHangtimeCameraLensEffect)
          (Ability|RemoveGameplayEffectFromOwnerWithHandle (Variables|Default|GetReticleShrinkEffect)))))))

(fn RemoveHangtimeCameraLensEffect
  (Camera|RemoveGenericCameraLensEffect (Variables|/Script/Engine.PlayerController|GetPlayerCameraManager (Fortnite|GetFortPlayerControllerfromActor (Variables|Default|GetAvatarFortPawn))) (Variables|Default|GetCameraLensEffect))
  (bind _outmodifier (Class|WeaponFunctionLibrary|FindorAddCameraModifierforPawn (Variables|Default|GetCamera%20Modifier%20Class) (Variables|Default|GetAvatarFortPawn)))
  (bind _asseven_slide_camera_modifier_hang_time (Utilities|Casting|CastToSevenSlide_CameraModifier_HangTime _outmodifier))
  (Variables|/SevenSlide/Vfx/SevenSlide_CameraModifier_HangTime.SevenSlide_CameraModifier_HangTime_C|SetIsActive false _asseven_slide_camera_modifier_hang_time)
  (Development|PrintString "remove camera lens"))
