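; Passive SevenSlide gameplay ability that enables instant crouch input for touch players, so crouch triggers as soon as the button is tapped.
; CanActivateAbility allows activation only when the avatar's player controller is using Touch input.
; Activation registers the ability on the FortPlayerPawn through SetUseInstantCrouchInput and prints a debug message.
; Ending the ability removes that registration through ClearUseInstantCrouchInput.
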(event Ability|EventActivateAbility
  (bind _self self)
  ; Crouch Instantly when button is tapped
  (bind _asfort_player_pawn (Utilities|Casting|CastToFortPlayerPawn (Ability|GetAvatarActorfromActorInfo)))
  (FortPawn|SetUseInstantCrouchInput _asfort_player_pawn _self)
  (Development|PrintText (Utilities|Text|FormatText "NSLOCTEXT(\"[180B4D726B8F11D4ECD2EC87B1809F63]\", \"ABE2AA644B2BE8C56A8DB79C9890D0F5\", \"{a} registered for Instant Crouch Input\")" _self)))

(event Ability|EventOnEndAbility
  (params (bWasCancelled :type Boolean))
  ; Removes this GA from instant crouch registry
  (bind _asfort_player_pawn (Utilities|Casting|CastToFortPlayerPawn (Ability|GetAvatarActorfromActorInfo)))
  (FortPawn|ClearUseInstantCrouchInput _asfort_player_pawn self))

(fn CanActivateAbility
  (params (ActorInfo :type Gameplay%20Ability%20Actor%20Info%20Structure) (Handle :type Gameplay%20Ability%20Spec%20Handle%20Structure))
  (return 0 (Inputs|IsUsingInputType "Touch" (Ability|GetPlayerControllerfromActorInfo))))
