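; Gameplay ability that lets a player in SevenSlide guided motion destroy props they hit, when destruction is enabled.
; Gameplay tag wait tasks bind the pawn's hit event while the guided motion tag is present and unbind it on removal.
; Hits filter for damageable, low-health, non-hiding building props and check the minigame's environment damage setting.
; The authority kills the prop, while clients predict its death.
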(event Ability|EventActivateAbility
  (if (ScalableFloat|GetValueasBool (Variables|Default|GetDestructionEnabled))
    (Ability|Tasks|WaitGameplayTagAdd "(TagName=\"MovementMode.SevenSlide.GuidedMotion\")"
      (:then
        (Ability|Tasks|WaitGameplayTagRemove "(TagName=\"MovementMode.SevenSlide.GuidedMotion\")"
          (:Removed
            (Collision|UnbindEventfromOnActorHit (CallFunction|GetAvatarasFortPlayerPawn) (delegate Custom|OnActorHit)))))
      (:Added
        (Collision|BindEventtoOnActorHit (CallFunction|GetAvatarasFortPlayerPawn) (delegate Custom|OnActorHit))))))

(event Custom|OnActorHit
  (params (SelfActor :type Actor%20Object%20Reference) (OtherActor :type Actor%20Object%20Reference) (NormalImpulse :type Vector) (Hit_bBlockingHit :type Boolean) (Hit_bInitialOverlap :type Boolean) (Hit_Time :type Float%20%28single-precision%29) (Hit_Distance :type Float%20%28single-precision%29) (Hit_Location :type Vector :default "0, 0, 0") (Hit_ImpactPoint :type Vector :default "0, 0, 0") (Hit_Normal :type Vector :default "0, 0, 0") (Hit_ImpactNormal :type Vector :default "0, 0, 0") (Hit_PhysMat :type Physical%20Material%20Object%20Reference) (Hit_HitActor :type Actor%20Object%20Reference) (Hit_HitComponent :type Primitive%20Component%20Object%20Reference) (Hit_HitBoneName :type Name :default "None") (Hit_BoneName :type Name :default "None") (Hit_HitItem :type Integer) (Hit_ElementIndex :type Integer) (Hit_FaceIndex :type Integer) (Hit_TraceStart :type Vector :default "0, 0, 0") (Hit_TraceEnd :type Vector :default "0, 0, 0"))
  (bind _returnvalue (CallFunction|GetAvatarasFortPlayerPawn))
  (bind _asbuilding_prop (Utilities|Casting|CastToBuildingProp OtherActor))
  (if (Utilities|IsA _asbuilding_prop "/Script/FortniteGame.FortHidingProp")
    _
    (elif (Math|Boolean|ANDBoolean (Class|BPFLAICommonFunctions|GetCanbeDamaged _asbuilding_prop) (.CanBeDamaged (Class|BPFLAICommonFunctions|GetCanbeDamaged _asbuilding_prop)) (< (Health|GetMaxHealth _asbuilding_prop :!class /Script/FortniteGame.BuildingActor) 121.0))
      (switch Utilities|FlowControl|Switch|SwitchonEnum_ExperienceSettings_EnvironmentDamageType (Utilities|Enum|BytetoEnumEnum_ExperienceSettings_EnvironmentDamageType (Math|Conversions|ToByte(Integer) (ScalableFloat|GetValueasIntegerwithContext (Class|GetClassDefaults "/Game/Athena/Playset/Minigames/Minigame_PlayerCreated.Minigame_PlayerCreated_C"))))
        (:NewEnumerator1
          _)
        (:NewEnumerator3
          (if (BuildingActor|IsPlayerBuilt _asbuilding_prop)
            _)))
      (Ability|HasAuthority
        (:True
          (Gameplay|Die _asbuilding_prop (Variables|Default|GetDeathTag) (Pawn|GetController _returnvalue) _returnvalue :!class /Script/FortniteGame.BuildingActor))
        (:False
          (Gameplay|PredictDeath _asbuilding_prop (Variables|Default|GetDeathTag) _returnvalue _returnvalue Hit_HitComponent true true true))))))

(event Ability|EventOnEndAbility
  (params (bWasCancelled :type Boolean)))
