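; Passive SevenSlider gameplay ability that swaps the player's sliding movement parameters while the device is powered and not overheated.
; Tag-wait tasks track water (with delayed exit), sliding, power and overheat, feeding CheckApplyActiveEffects, which selects default, normal or in-water overrides without newly applying them mid-slide.
; The server adds the overheat tag at empty fuel and removes it at full.
; Fuel-level loop audio accompanies sliding, with client-side, cooldown-limited low-fuel warnings.
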(event Ability|EventActivateAbility
  (bind (_bsuccessfullyfoundattribute _returnvalue) (Ability|Attribute|GetFloatAttributefromAbilitySystemComponent (Variables|Default|GetAvatarAbilitySystem) "(AttributeName=\"MaxFuel\",Attribute=/Script/SevenSlideRuntime.SevenSlideAttributeSet:MaxFuel,AttributeOwner=\"/Script/CoreUObject.Class'/Script/SevenSlideRuntime.SevenSlideAttributeSet'\")"))
  (Variables|Default|SetAvatarAbilitySystem (Ability|GetAbilitySystemComponent (CallFunction|GetAvatarasFortPlayerPawn)))
  (Variables|Default|SetAppliedSlideParams "NewEnumerator0")
  (Variables|Default|SetAppliedOverheatTag false)
  (Variables|Default|SetOverheated false)
  (Variables|Default|SetPoweredOn true)
  (Utilities|FlowControl|Sequence
    (:then_0
      ; Monitor Tags
      (bind _asynctaskproxy (Ability|Tasks|WaitGameplayTagAdd "(TagName=\"Gameplay.Status.InWaterBody\")")
        (:then
          (Variables|AsyncHandles|SetInWaterTask _asynctaskproxy))
        (:Added
          (Variables|Default|SetIsInWater true)
          (Utilities|FlowControl|Sequence
            (:then_0
              (bind _outofwaterdelaytask (Variables|AsyncHandles|GetOutOfWaterDelayTask)
                (:then
                  (GameplayTasks|EndTask _outofwaterdelaytask)
                  (Variables|AsyncHandles|SetOutOfWaterDelayTask 0))
                (:else)))
            (:then_1
              (CallFunction|CheckApplyActiveEffects))))))
    (:then_1
      (bind _asynctaskproxy_2 (Ability|Tasks|WaitGameplayTagRemove "(TagName=\"Gameplay.Status.InWaterBody\")")
        (:then
          (Variables|AsyncHandles|SetOutOfWaterTask _asynctaskproxy_2))
        (:Removed
          (bind _asynctaskproxy_3 (Ability|Tasks|WaitDelay 0.5)
            (:then
              (Variables|AsyncHandles|SetOutOfWaterDelayTask _asynctaskproxy_3))
            (:OnFinish
              (Variables|Default|SetIsInWater false)
              (CallFunction|CheckApplyActiveEffects))))))
    (:then_2
      (bind _asynctaskproxy_4 (Ability|Tasks|WaitGameplayTagAdd "(TagName=\"MovementMode.Sliding\")")
        (:then
          (Variables|AsyncHandles|SetNativeSlidingTask _asynctaskproxy_4))
        (:Added
          (Variables|Default|SetIsNativeSliding true)
          (CallFunction|CheckApplyActiveEffects)
          (CallFunction|PlayFuelAudio))))
    (:then_3
      (bind _asynctaskproxy_5 (Ability|Tasks|WaitGameplayTagRemove "(TagName=\"MovementMode.Sliding\")")
        (:then
          (Variables|AsyncHandles|SetNativeSlidingDoneTask _asynctaskproxy_5))
        (:Removed
          (Variables|Default|SetIsNativeSliding false)
          (CallFunction|CheckApplyActiveEffects)
          (CallFunction|FadeOutFuelAudio))))
    (:then_4
      (bind _asynctaskproxy_6 (Ability|Tasks|WaitGameplayTagAdd "(TagName=\"sevenslide.PoweredOn\")")
        (:then
          (Variables|AsyncHandles|SetPoweredOnTask _asynctaskproxy_6))
        (:Added
          (Variables|Default|SetPoweredOn true)
          (CallFunction|CheckApplyActiveEffects))))
    (:then_5
      (bind _asynctaskproxy_7 (Ability|Tasks|WaitGameplayTagRemove "(TagName=\"sevenslide.PoweredOn\")")
        (:then
          (Variables|AsyncHandles|SetPoweredOffTask _asynctaskproxy_7))
        (:Removed
          (Variables|Default|SetPoweredOn false)
          (CallFunction|CheckApplyActiveEffects))))
    (:then_6
      (bind _asynctaskproxy_8 (Ability|Tasks|WaitGameplayTagAdd "(TagName=\"sevenslide.Overheat\")")
        (:then
          (Variables|AsyncHandles|SetOverheatedTask _asynctaskproxy_8))
        (:Added
          (Variables|Default|SetOverheated true)
          (CallFunction|CheckApplyActiveEffects))))
    (:then_7
      (bind _asynctaskproxy_9 (Ability|Tasks|WaitGameplayTagRemove "(TagName=\"sevenslide.Overheat\")")
        (:then
          (Variables|AsyncHandles|SetOverheatDoneTask _asynctaskproxy_9))
        (:Removed
          (Variables|Default|SetOverheated false)
          (CallFunction|CheckApplyActiveEffects))))
    (:then_8
      ; On server, monitor overheat
      (Ability|HasAuthority
        (:True
          (bind (_asynctaskproxy_10 _bmatchescomparison) (Ability|Tasks|WaitForAttributeChangeThreshold "(AttributeName=\"Fuel\",Attribute=/Script/SevenSlideRuntime.SevenSlideAttributeSet:Fuel,AttributeOwner=\"/Script/CoreUObject.Class'/Script/SevenSlideRuntime.SevenSlideAttributeSet'\")" "GreaterThanOrEqualTo" _returnvalue)
            (:then
              (Variables|AsyncHandles|SetFullFuelTask _asynctaskproxy_10))
            (:OnChange
              (if _bmatchescomparison
                (Variables|Default|GetAppliedOverheatTag
                  (:then
                    (Variables|Default|SetAppliedOverheatTag false)
                    (Ability|GameplayEffect|RemoveLooseGameplayTags (CallFunction|GetAvatarasFortPlayerPawn) (Variables|Default|GetOverheatTag) true))
                  (:else))))))))
    (:then_9
      (Ability|HasAuthority
        (:True
          (bind (_asynctaskproxy_11 _bmatchescomparison_2) (Ability|Tasks|WaitForAttributeChangeThreshold "(AttributeName=\"Fuel\",Attribute=/Script/SevenSlideRuntime.SevenSlideAttributeSet:Fuel,AttributeOwner=\"/Script/CoreUObject.Class'/Script/SevenSlideRuntime.SevenSlideAttributeSet'\")" "LessThanOrEqualTo")
            (:then
              (Variables|AsyncHandles|SetEmptyFuelTask _asynctaskproxy_11))
            (:OnChange
              (if _bmatchescomparison_2
                (Variables|Default|GetAppliedOverheatTag
                  (:then)
                  (:else
                    (Variables|Default|SetAppliedOverheatTag true)
                    (Ability|GameplayEffect|AddLooseGameplayTags (CallFunction|GetAvatarasFortPlayerPawn) (Variables|Default|GetOverheatTag) true)))))))))
    (:then_10
      ; On client, low fuel warning
      (Ability|HasAuthority
        (:False
          (Variables|Default|SetcurrentTimeToCooldownTo (Utilities|Time|GetGameTimeinSeconds))
          (bind (_asynctaskproxy_12 _bmatchescomparison_3) (Ability|Tasks|WaitForAttributeChangeThreshold "(AttributeName=\"Fuel\",Attribute=/Script/SevenSlideRuntime.SevenSlideAttributeSet:Fuel,AttributeOwner=\"/Script/CoreUObject.Class'/Script/SevenSlideRuntime.SevenSlideAttributeSet'\")" "LessThanOrEqualTo" (ScalableFloat|GetValueatLevel (Variables|Default|GetHF_LowFuelWarningFuelAmount)))
            (:then
              (Variables|AsyncHandles|SetLowFuelTask _asynctaskproxy_12))
            (:OnChange
              (if _bmatchescomparison_3
                (if (GameplayTags|HasMatchingGameplayTag (CallFunction|GetAvatarasFortPlayerPawn) "(TagName=\"MovementMode.Falling\")")
                  _
                  (elif (and (Variables|Default|GetIsNativeSliding) (>= (Utilities|Time|GetGameTimeinSeconds) (Variables|Default|GetcurrentTimeToCooldownTo)))
                    (Audio|SpawnSoundAttached "/SevenSlide/Assets/Audio/MSS_SevenSlide_Fuel_Warning.MSS_SevenSlide_Fuel_Warning" (Variables|/Script/Engine.Character|GetMesh (CallFunction|GetAvatarasFortPlayerPawn)))
                    (Variables|Default|SetcurrentTimeToCooldownTo (+ (Variables|Default|GetcooldownTime) (Utilities|Time|GetGameTimeinSeconds)))))))))))))

(event Custom|PlayFuelAudio
  ; On client, low fuel warning
  (if (not (Utilities|Enum|Equal(Enum) (Variables|Default|GetAppliedSlideParams) "NewEnumerator0"))
    (bind _returnvalue (Utilities|Time|SetTimerbyEvent (delegate Custom|UpdateFuelLevelForAudio) 0.5 true true))
    (Variables|Default|SetFuelLevelAudioTimer _returnvalue)))

(event Custom|UpdateFuelLevelForAudio
  (bind (_bsuccessfullyfoundattribute _returnvalue) (Ability|Attribute|GetFloatAttribute (Variables|Components|GetAvatarPlayerPawn) "(AttributeName=\"Fuel\",Attribute=/Script/SevenSlideRuntime.SevenSlideAttributeSet:Fuel,AttributeOwner=\"/Script/CoreUObject.Class'/Script/SevenSlideRuntime.SevenSlideAttributeSet'\")"))
  ; On client, low fuel warning
  (Variables|Default|GetOverheated
    (:then
      (CallFunction|FadeOutFuelAudio))
    (:else
      (bind _fuellevelaudio (Variables|Default|GetFuelLevelAudio)
        (:then
          (Audio|Parameter|SetFloatParameter _fuellevelaudio "FuelLevel" _returnvalue :!class /Script/AudioExtensions.AudioParameterControllerInterface))
        (:else
          (bind _returnvalue_2 (Audio|SpawnSoundAttached "/SevenSlide/Assets/Audio/MSS_SevenSlide_FuelLevel_Loop.MSS_SevenSlide_FuelLevel_Loop" (Variables|/Script/Engine.Character|GetMesh (CallFunction|GetAvatarasFortPlayerPawn))))
          (Variables|Default|SetFuelLevelAudio _returnvalue_2))))))

(event Custom|FadeOutFuelAudio
  ; On client, low fuel warning
  (Utilities|Time|ClearandInvalidateTimerbyHandle (Variables|Default|GetFuelLevelAudioTimer))
  (bind _fuellevelaudio (Variables|Default|GetFuelLevelAudio)
    (:then
      (Audio|Components|Audio|FadeOut _fuellevelaudio 0.5 0.0 "Linear" :!class /Script/Engine.AudioComponent))
    (:else)))

(event Ability|EventOnEndAbility
  (params (bWasCancelled :type Boolean))
  ; End Ability
  (Utilities|FlowControl|Sequence
    (:then_0
      (bind _outofwaterdelaytask (Variables|AsyncHandles|GetOutOfWaterDelayTask)
        (:then
          (GameplayTasks|EndTask _outofwaterdelaytask)
          (Variables|AsyncHandles|SetOutOfWaterDelayTask 0))
        (:else)))
    (:then_1
      (bind _inwatertask (Variables|AsyncHandles|GetInWaterTask)
        (:then
          (GameplayTasks|EndTask _inwatertask)
          (Variables|AsyncHandles|SetInWaterTask 0))
        (:else)))
    (:then_2
      (bind _outofwatertask (Variables|AsyncHandles|GetOutOfWaterTask)
        (:then
          (GameplayTasks|EndTask _outofwatertask)
          (Variables|AsyncHandles|SetOutOfWaterTask 0))
        (:else)))
    (:then_3
      (bind _poweredontask (Variables|AsyncHandles|GetPoweredOnTask)
        (:then
          (GameplayTasks|EndTask _poweredontask)
          (Variables|AsyncHandles|SetPoweredOnTask 0))
        (:else)))
    (:then_4
      (bind _poweredofftask (Variables|AsyncHandles|GetPoweredOffTask)
        (:then
          (GameplayTasks|EndTask _poweredofftask)
          (Variables|AsyncHandles|SetPoweredOffTask 0))
        (:else)))
    (:then_5
      (bind _overheatedtask (Variables|AsyncHandles|GetOverheatedTask)
        (:then
          (GameplayTasks|EndTask _overheatedtask)
          (Variables|AsyncHandles|SetOverheatedTask 0))
        (:else)))
    (:then_6
      (bind _overheatdonetask (Variables|AsyncHandles|GetOverheatDoneTask)
        (:then
          (GameplayTasks|EndTask _overheatdonetask)
          (Variables|AsyncHandles|SetOverheatDoneTask 0))
        (:else)))
    (:then_7
      (bind _fullfueltask (Variables|AsyncHandles|GetFullFuelTask)
        (:then
          (GameplayTasks|EndTask _fullfueltask)
          (Variables|AsyncHandles|SetFullFuelTask 0))
        (:else)))
    (:then_8
      (bind _emptyfueltask (Variables|AsyncHandles|GetEmptyFuelTask)
        (:then
          (GameplayTasks|EndTask _emptyfueltask)
          (Variables|AsyncHandles|SetEmptyFuelTask 0))
        (:else)))
    (:then_9
      (bind _lowfueltask (Variables|AsyncHandles|GetLowFuelTask)
        (:then
          (GameplayTasks|EndTask _lowfueltask)
          (Variables|AsyncHandles|SetLowFuelTask 0))
        (:else)))
    (:then_10
      (switch Utilities|FlowControl|Switch|SwitchonE_SevenSlideParamOverrideMode (Variables|Default|GetAppliedSlideParams)
        (:NewEnumerator1
          _)
        (:NewEnumerator2
          _))
      (Variables|Default|SetAppliedSlideParams "NewEnumerator0")
      (CallFunction|ClearSlideControlOverrides))
    (:then_11
      (Variables|Default|GetAppliedOverheatTag
        (:then
          (Variables|Default|SetAppliedOverheatTag false)
          (Ability|GameplayEffect|RemoveLooseGameplayTags (CallFunction|GetAvatarasFortPlayerPawn) (Variables|Default|GetOverheatTag) true))
        (:else)))
    (:then_12
      (CallFunction|FadeOutFuelAudio))))

(fn ApplySlideControlOverrides
  (params (UseInWaterOverrides :type Boolean))
  (bind _asfort_player_pawn (Utilities|Casting|CastToFortPlayerPawn (Ability|GetAvatarActorfromActorInfo)))
  (bind _output_get (Variables|Components|SetAvatarPlayerPawn _asfort_player_pawn))
  (bind _asfort_movement_comp_character (Utilities|Casting|CastToFortMovementComp_Character (Variables|/Script/Engine.Character|GetCharacterMovement _output_get)))
  (bind _output_get_2 (Variables|Default|SetAvatarMovementComp _asfort_movement_comp_character))
  (bind _output_get_3 (LocalVariables|SetSlidingParams (Variables|/Script/FortniteGame.FortMovementComp_Character|GetSlidingControlParams _output_get_2)))
  (bind _structout (Utilities|Struct|SetmembersinSlidingControlParams _output_get_3 (Variables|SlideParams|GetBreakingDeceleration) (Variables|SlideParams|GetMin%20Speed%20To%20Start%20Sliding) (Variables|SlideParams|GetUphill%20Slide%20Start%20Boost) (Variables|SlideParams|GetLevel%20Slide%20Start%20Boost) (Variables|SlideParams|GetDownhill%20Slide%20Start%20Boost) (Variables|SlideParams|GetBoost%20Level%20Ground%20Friction%20Multiplier) (Variables|SlideParams|GetUphill%20Ground%20Friction%20Multiplier) (Variables|SlideParams|GetLevel%20Ground%20Friction%20Multiplier) (Variables|SlideParams|GetDownhill%20Ground%20Friction%20Multiplier) (Variables|SlideParams|GetMax%20Forward%20Sliding%20Speed) (Variables|SlideParams|GetMin%20Forward%20Sliding%20Speed) (Variables|SlideParams|GetMaxForwardAcceleration) (Variables|SlideParams|GetSlidingFatigueDuration) (Variables|SlideParams|GetSlidingFatigueMutliplier) (select (Variables|GetUseInWaterOverrides) (Variables|SlideParams|GetInWaterFallingGracePeriod) (Variables|SlideParams|GetFallingGracePeriod))))
  (Variables|/Script/FortniteGame.FortMovementComp_Character|SetSlidingControlParams _structout (Variables|Default|GetAvatarMovementComp))
  (if (> (ScalableFloat|GetValueasDouble (Variables|Default|GetAllowSlidingInWaterHotfix)) 0)
    (Ability|GameplayEffect|AddLooseGameplayTags (Variables|Components|GetAvatarPlayerPawn) (Variables|Default|GetActivePawnTags))))

(fn CheckApplyActiveEffects
  (LocalVariables|SetAreAnyParamsApplied (Utilities|Enum|NotEqual(Enum) (Variables|Default|GetAppliedSlideParams) "NewEnumerator0"))
  (Utilities|FlowControl|Sequence
    (:then_0
      ; Determine new param override state
      (if (or (Variables|Default|GetOverheated) (not (Variables|Default|GetPoweredOn)))
        (LocalVariables|SetSlideParamToBeApplied "NewEnumerator0")
        (elif (and (Variables|Default|GetIsNativeSliding) (not (LocalVariables|GetAreAnyParamsApplied)))
          (LocalVariables|SetSlideParamToBeApplied "NewEnumerator0")
          (else
            (Variables|Default|GetIsInWater
              (:then
                (LocalVariables|SetSlideParamToBeApplied "NewEnumerator2"))
              (:else
                (LocalVariables|SetSlideParamToBeApplied "NewEnumerator1")))))))
    (:then_1
      ; Apply param override
      (if (Utilities|Enum|NotEqual(Enum) (LocalVariables|GetSlideParamToBeApplied) (Variables|Default|GetAppliedSlideParams))
        (bind _output_get (Variables|Default|SetAppliedSlideParams (LocalVariables|GetSlideParamToBeApplied)))
        (switch Utilities|FlowControl|Switch|SwitchonE_SevenSlideParamOverrideMode _output_get
          (:NewEnumerator0
            (CallFunction|ClearSlideControlOverrides))
          (:NewEnumerator1
            (CallFunction|ApplySlideControlOverrides false))
          (:NewEnumerator2
            (CallFunction|ApplySlideControlOverrides true)))))))

(fn ClearSlideControlOverrides
  (Utilities|FlowControl|Sequence
    (:then_0
      (bind _avatarmovementcomp (Variables|Default|GetAvatarMovementComp)
        (:then
          (Class|FortMovementCompCharacter|RestoreDefaultSlidingControlParams _avatarmovementcomp)
          (Variables|Default|SetAppliedSlideParams "NewEnumerator0"))
        (:else)))
    (:then_1
      (if (ScalableFloat|GetValueasBool (Variables|Default|GetHF_SlideOnWalkingTapCrouchInput))
        (bind _avatarplayerpawn (Variables|Components|GetAvatarPlayerPawn)
          (:then
            (FortPawn|ClearUseInstantCrouchInput _avatarplayerpawn self))
          (:else))))
    (:then_2
      (bind _avatarplayerpawn_2 (Variables|Components|GetAvatarPlayerPawn)
        (:then
          (Ability|GameplayEffect|RemoveLooseGameplayTags _avatarplayerpawn_2 (Variables|Default|GetActivePawnTags)))
        (:else)))))
