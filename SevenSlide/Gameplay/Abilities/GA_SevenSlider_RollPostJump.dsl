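; Gameplay ability for the SevenSlide post-jump roll phase, active until the pawn returns to the ground.
; Activation applies the activation cost effect and binds pawn-landed and movement-mode-changed callbacks.
; Landing or switching to walking or nav walking adds the exit tag and ends the ability, unless the boosted jump movement tag is present.
; Ending removes the RollPostJump effect, unbinds both callbacks and clears a durability timer handle that this graph never sets.
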(event Ability|EventActivateAbility
  (Ability|ApplyGameplayEffectToOwner "/SevenSlide/Gameplay/Effects/GE_SevenSlideActivationCost.GE_SevenSlideActivationCost_C") ; Activation Cost
  ;[1 1 1] Set Up Binds
  (bind _asfort_pawn (Utilities|Casting|CastToFortPawn (Ability|GetAvatarActorfromActorInfo)))
  (EventDispatchers|BindEventtoOnPawnLanded _asfort_pawn (delegate Custom|PawnLanded))
  (Character|BindEventtoMovementModeChangedDelegate _asfort_pawn (delegate Custom|MovementModeChanged)))

(event Custom|MovementModeChanged
  (params (Character :type Character%20Object%20Reference) (PrevMovementMode :type EMovementMode%20Enum :default "MOVE_None") (PreviousCustomMode :type Byte))
  ;[1 1 1] Set Up Binds
  (if (GameplayTags|HasMatchingGameplayTag (CallFunction|GetAvatarasFortPlayerPawn) "(TagName=\"MovementMode.SevenSlide.BoostedJumpMovement\")")
    _
    (elif (or (Utilities|Enum|Equal(Enum) (Variables|/Script/Engine.CharacterMovementComponent|GetMovementMode (Variables|/Script/Engine.Character|GetCharacterMovement Character)) "MOVE_Walking") (Utilities|Enum|Equal(Enum) (Variables|/Script/Engine.CharacterMovementComponent|GetMovementMode (Variables|/Script/Engine.Character|GetCharacterMovement Character)) "MOVE_NavWalking"))
      (Ability|GameplayEffect|AddLooseGameplayTags (CallFunction|GetAvatarasFortPlayerPawn) (Variables|Default|GetExitTag) true)
      (Ability|EndAbility))))

(event Custom|PawnLanded
  (params (Hit :type Hit%20Result%20Structure%20%28by%20ref%29))
  (if (not (GameplayTags|HasMatchingGameplayTag (CallFunction|GetAvatarasFortPlayerPawn) "(TagName=\"MovementMode.SevenSlide.BoostedJumpMovement\")"))
    (Ability|GameplayEffect|AddLooseGameplayTags (CallFunction|GetAvatarasFortPlayerPawn) (Variables|Default|GetExitTag) true)
    ;[1 1 1] Set Up Binds
    (Ability|EndAbility)))

(event Ability|EventOnEndAbility
  (params (bWasCancelled :type Boolean))
  ;[1 1 1] Remove Binds
  (GameplayEffects|RemoveActiveGameplayEffectbySourceEffect (Ability|GetAbilitySystemComponent (CallFunction|GetAvatarasFortPlayerPawn)) "/SevenSlide/Gameplay/Effects/GE_SevenSlider_RollPostJump.GE_SevenSlider_RollPostJump_C" (Ability|GetAbilitySystemComponent (CallFunction|GetAvatarasFortPlayerPawn)) -1)
  (bind _asfort_pawn (Utilities|Casting|CastToFortPawn (Ability|GetAvatarActorfromActorInfo)))
  (EventDispatchers|UnbindEventfromOnPawnLanded _asfort_pawn (EventDispatchers|CreateEvent :!event Custom|PawnLanded))
  (Character|UnbindEventfromMovementModeChangedDelegate _asfort_pawn (EventDispatchers|CreateEvent :!event Custom|MovementModeChanged))
  (Utilities|Time|ClearandInvalidateTimerbyHandle (Variables|Default|GetDurabiliityTick)))
