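; Gameplay ability for the SevenSlider flame tail visual, driven by the sevenslide.activewindow tag.
; Tag-added starts a fast looping timer meant to spawn flames, though no spawn delegate or array fill appears here.
; Tag-removed stops that timer and schedules delayed deactivation and destruction of the components in FireTailArray.
; Ability end terminates both tag-wait tasks, clears the timers and runs the cleanup immediately.
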(event Ability|EventActivateAbility
  (bind _asynctaskproxy (Ability|Tasks|WaitGameplayTagAdd "(TagName=\"sevenslide.activewindow\")")
    (:then
      (Variables|Default|SetAddedTimer _asynctaskproxy)
      (bind _asynctaskproxy_2 (Ability|Tasks|WaitGameplayTagRemove "(TagName=\"sevenslide.activewindow\")")
        (:then
          (Variables|Default|SetRemoveTimer _asynctaskproxy_2))
        (:Removed
          (Utilities|Time|ClearandInvalidateTimerbyHandle (Variables|Default|GetTailVisualTimer))
          (bind _returnvalue (Utilities|Time|SetTimerbyEvent (delegate Custom|DeleteVfx) 3.0))
          (Variables|Default|SetCleanUpDelay _returnvalue))))
    (:Added
      ; TODO temporary spawn flames behind the slide
      (bind _returnvalue_2 (Utilities|Time|SetTimerbyEvent 0 0.07 true true))
      (Variables|Default|SetTailVisualTimer _returnvalue_2))))

(event Custom|DeleteVfx
  ; Clean up visual effects
  (for _array_element (Variables|Default|GetFireTailArray)
    (Components|Activation|Deactivate _array_element)
    (Components|DestroyComponent _array_element))
  (Utilities|Array|Clear (Variables|Default|GetFireTailArray)))

(event Ability|EventOnEndAbility
  (params (bWasCancelled :type Boolean))
  (Utilities|Time|ClearandInvalidateTimerbyHandle (Variables|Default|GetCleanUpDelay))
  (GameplayTasks|EndTask (Variables|Default|GetAddedTimer))
  (GameplayTasks|EndTask (Variables|Default|GetRemoveTimer))
  (Utilities|Time|ClearandInvalidateTimerbyHandle (Variables|Default|GetTailVisualTimer))
  (Utilities|Time|ClearandInvalidateTimerbyHandle (Variables|Default|GetCleanUpDelay))
  (CallFunction|DeleteVfx))
