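; Gameplay ability for SevenSlide that manages player tags and effects based on the held weapon and attacking.
; It tracks the current weapon through a WeaponSwitch event task, which it ends with the ability.
; It clears equipped tags, then adds loose player tags from weapon-to-player tag pairs, falling back to OtherTag.
; It also applies a RecentlyAttacked effect when the Attacking tag appears, and a building-block effect gated by a hotfix scalar.
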(event Ability|EventActivateAbility
  (bind _currentweapon (Variables|Default|GetCurrentWeapon))
  (Ability|Tasks|WaitGameplayTagAdd "(TagName=\"Gameplay.Action.Attacking\")"
    (:then
      ; Disallow Building as a hotfix
      (if (not (> (ScalableFloat|GetValueasDouble (Variables|Default|GetBuildingAllowedWhileSlidingHotfix)) 0))
        (Ability|ApplyGameplayEffectToOwner "/SevenSlide/Gameplay/Effects/GE_BoostSeven_BlockBuildingHotfix.GE_BoostSeven_BlockBuildingHotfix_C"))
      (bind (_asynctaskproxy _payload) (Ability|Tasks|WaitGameplayEvent "(TagName=\"Event.Weapons.WeaponSwitch\")")
        (:then
          (Variables|Default|SetEventWatcher _asynctaskproxy)
          (Variables|Default|SetCurrentWeapon (Variables|/Script/FortniteGame.FortPawn|GetCurrentWeapon (CallFunction|GetAvatarasFortPlayerPawn))))
        (:EventReceived
          (bind (_payload_eventtag _payload_instigator _payload_target _payload_optionalobject) (Utilities|Struct|BreakGameplayEventData _payload))
          (bind _asfort_weapon (Utilities|Casting|CastToFortWeapon _payload_optionalobject))
          (Variables|Default|SetCurrentWeapon _asfort_weapon)))
      (Ability|GameplayEffect|RemoveLooseGameplayTags (CallFunction|GetAvatarasFortPlayerPawn) (Variables|Default|GetEquippedGameplayTags))
      (Variables|Default|SetFoundTag false)
      (for _array_element (Variables|Default|GetWeaponTagsToPlayerTags)
        (bind (_array_element_weapontag_5_0389b4f848626c1830d01fad3baee04a _array_element_playertag_6_6adaf93742417c974055518c9203f4bf) (Utilities|Struct|BreakStructSevenSlideTagPair _array_element))
        (Utilities|IsValid _currentweapon
          (:Is%20Valid
            (if (Tags|WeaponDataHasAnyMatchingGameplayTags _currentweapon _array_element_weapontag_5_0389b4f848626c1830d01fad3baee04a)
              (Ability|GameplayEffect|AddLooseGameplayTags (CallFunction|GetAvatarasFortPlayerPawn) _array_element_playertag_6_6adaf93742417c974055518c9203f4bf)
              (Variables|Default|SetFoundTag true)))))
      (if (not (Variables|Default|GetFoundTag))
        (Ability|GameplayEffect|AddLooseGameplayTags (CallFunction|GetAvatarasFortPlayerPawn) (Variables|Default|GetOtherTag))))
    (:Added
      (Ability|ApplyGameplayEffectToOwner "/SevenSlide/Gameplay/Effects/GE_BoostSeven_RecentlyAttacked.GE_BoostSeven_RecentlyAttacked_C"))))

(event Ability|EventOnEndAbility
  (params (bWasCancelled :type Boolean))
  (GameplayTasks|EndTask (Variables|Default|GetEventWatcher)))
